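# Control Rig for the F_MED_EagerTrapper character body that adds corrective deformation and secondary skirt, belt and gear posing.
# Forward solve runs the shared DeformRig function, then Skirt_SPR, then Belt_Constraints.
# Skirt_SPR uses four spherical pose readers per thigh to weight skirt and belt bone offsets, with center skirt taking left/right maximums.
# Belt_Constraints blend-parents grenades to dynamic belt bones and the knee pad to leg bones, adding an attribute-driven knee offset.
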ControlRigBlueprint /BRCosmetics/Characters/Player/Female/Medium/Bodies/F_MED_EagerTrapper/Meshes/F_MED_EagerTrapper_CtrlRig
graphs [
 Maximum_2_ContainedGraph {nodes [
   n1 = RigVMFunctionEntryNode()
   n2 = &RigVMFunction_MathDoubleMax(A=@n1.A, B=@n1.B)
   RigVMFunctionReturnNode(Result=&RigVMFunction_MathDoubleMax(B=@n1.D, A=&RigVMFunction_MathDoubleMax(B=@n1.C, A=@n2.Result).Result).Result)
  ]}
 New Function_ContainedGraph {nodes [
   RigVMFunctionEntryNode()
   > RigVMFunction_Sequence = RigVMAggregateNode()
   n1 = &RigUnit_SphericalPoseReader(
    ActiveRegionScaleFactors=(PositiveWidth=1,NegativeWidth=1,PositiveHeight=1,NegativeHeight=1)
    ActiveRegionSize=0.05
    Debug=(bDrawDebug=false,bDraw2D=False,bDrawLocalAxes=False,DebugScale=25,DebugSegments=20,DebugThickness=0.25)
    DriverAxis=(X=1,Y=0,Z=0)
    DriverItem=(Type=Bone,Name="thigh_r")
    FalloffRegionScaleFactors=(PositiveWidth=1,NegativeWidth=1,PositiveHeight=1,NegativeHeight=1)
    FalloffSize=0.55
    FlipHeightScaling=true
    FlipWidthScaling=true
    OptionalParentItem=(Type=Bone,Name="None")
    RotationOffset=(X=-90,Y=0,Z=-15)
    ExecutePin=@RigVMFunction_Sequence.A
   )
   > &RigUnit_ModifyTransforms(
    ItemToModify="((Item=(Type=Bone,Name=\"base_r_skirt_A\"),Transform=(Rotation=(X=-0.1607,Y=-0.6707,Z=0.4007,W=0.6032),Translation=(X=10.2316,Y=-1.3933,Z=0.4354),Scale3D=(X=1,Y=1,Z=1))),(Item=(Type=Bone,Name=\"base_r_skirt_B\"),Transform=(Rotation=(X=0,Y=0,Z=0.5,W=0.866),Tran ...(truncated)"
    WeightMinimum=0
    Weight=@n1.OutputParam
   )
   > &RigUnit_ModifyTransforms(
    ItemToModify="((Item=(Type=Bone,Name=\"base_belt_front_r\"),Transform=(Rotation=(X=0.142,Y=0.19,Z=-0.85,W=-0.4703),Translation=(X=-15.1322,Y=-1.8001,Z=-1.5302),Scale3D=(X=1,Y=1,Z=1))),(Item=(Type=Bone,Name=\"base_belt_side_r\"),Transform=(Rotation=(X=-0.0198,Y=-0.2322,Z=0.0 ...(truncated)"
    WeightMinimum=0
    Weight=@n1.OutputParam
   )
   > n2 = &RigUnit_SphericalPoseReader(
    ActiveRegionScaleFactors=(PositiveWidth=1,NegativeWidth=1,PositiveHeight=1,NegativeHeight=1)
    ActiveRegionSize=0.05
    Debug=(bDrawDebug=false,bDraw2D=False,bDrawLocalAxes=False,DebugScale=25,DebugSegments=20,DebugThickness=0.25)
    DriverAxis=(X=1,Y=0,Z=0)
    DriverItem=(Type=Bone,Name="thigh_r")
    FalloffRegionScaleFactors=(PositiveWidth=1,NegativeWidth=1,PositiveHeight=1,NegativeHeight=1)
    FalloffSize=0.55
    FlipHeightScaling=true
    FlipWidthScaling=true
    OptionalParentItem=(Type=Bone,Name="None")
    RotationOffset=(X=-40,Y=20,Z=0)
   )
   > &RigUnit_ModifyTransforms(
    ItemToModify="((Item=(Type=Bone,Name=\"base_r_skirt_A\"),Transform=(Rotation=(X=0.0029,Y=-0.0352,Z=-0.0514,W=0.998),Translation=(X=0.6188,Y=-0.0085,Z=-1.6172),Scale3D=(X=1,Y=1,Z=1))),(Item=(Type=Bone,Name=\"base_r_skirt_B\"),Transform=(Rotation=(X=-0.1536,Y=-0.1259,Z=0.0149 ...(truncated)"
    WeightMinimum=0
    Weight=@n2.OutputParam
   )
   > n3 = &RigUnit_SphericalPoseReader(
    ActiveRegionScaleFactors=(PositiveWidth=1,NegativeWidth=1,PositiveHeight=1,NegativeHeight=1)
    ActiveRegionSize=0.05
    Debug=(bDrawDebug=false,bDraw2D=False,bDrawLocalAxes=False,DebugScale=25,DebugSegments=20,DebugThickness=0.25)
    DriverAxis=(X=1,Y=0,Z=0)
    DriverItem=(Type=Bone,Name="thigh_r")
    FalloffRegionScaleFactors=(PositiveWidth=1,NegativeWidth=1,PositiveHeight=1,NegativeHeight=1)
    FalloffSize=0.55
    FlipHeightScaling=true
    FlipWidthScaling=true
    OptionalParentItem=(Type=Bone,Name="None")
    RotationOffset=(X=0,Y=20,Z=0)
   )
   > &RigUnit_ModifyTransforms(
    ItemToModify="((Item=(Type=Bone,Name=\"offset_r_skirt_B\"),Transform=(Rotation=(X=-0.0298,Y=-0.6684,Z=0.0331,W=0.7424),Translation=(X=12.7276,Y=0,Z=-3.703),Scale3D=(X=1,Y=1,Z=1))),(Item=(Type=Bone,Name=\"offset_r_skirt_A\"),Transform=(Rotation=(X=-0.2948,Y=-0.5201,Z=-0.383, ...(truncated)"
    WeightMinimum=0
    Weight=@n3.OutputParam
   )
   Multiply = &RigVMFunction_MathDoubleMul(B=1.4, A=@n3.OutputParam)
   > &RigUnit_ModifyTransforms(
    ItemToModify="((Item=(Type=Bone,Name=\"base_belt_front_r\"),Transform=(Rotation=(X=0,Y=-0.2577,Z=0,W=-0.9662),Translation=(X=-8.6953,Y=0,Z=0.6668),Scale3D=(X=1,Y=1,Z=1))),(Item=(Type=Bone,Name=\"base_belt_side_r\"),Transform=(Rotation=(X=-0.0532,Y=0.0663,Z=-0.9318,W=-0.3528 ...(truncated)"
    WeightMinimum=0
    Weight=@Multiply.Result
   )
   > n4 = &RigUnit_SphericalPoseReader(
    ActiveRegionScaleFactors=(PositiveWidth=1,NegativeWidth=1,PositiveHeight=1,NegativeHeight=1)
    ActiveRegionSize=0.05
    Debug=(bDrawDebug=false,bDraw2D=False,bDrawLocalAxes=False,DebugScale=25,DebugSegments=20,DebugThickness=0.25)
    DriverAxis=(X=1,Y=0,Z=0)
    DriverItem=(Type=Bone,Name="thigh_r")
    FalloffRegionScaleFactors=(PositiveWidth=1,NegativeWidth=1,PositiveHeight=1,NegativeHeight=1)
    FalloffSize=0.55
    FlipHeightScaling=False
    FlipWidthScaling=False
    OptionalParentItem=(Type=Bone,Name="None")
    RotationOffset=(X=90,Y=0,Z=20)
   )
   > &RigUnit_ModifyTransforms(
    ItemToModify="((Item=(Type=Bone,Name=\"base_r_skirt_C\"),Transform=(Rotation=(X=0.0177,Y=-0.7184,Z=-0.2149,W=0.6614),Translation=(X=7.7893,Y=1.511,Z=1.2419),Scale3D=(X=1,Y=1,Z=1))),(Item=(Type=Bone,Name=\"base_r_skirt_B\"),Transform=(Rotation=(X=-0.0586,Y=-0.118,Z=-0.5483,W ...(truncated)"
    WeightMinimum=0
    Weight=@n4.OutputParam
   )
   n5 = &RigUnit_SphericalPoseReader(
    ActiveRegionScaleFactors=(PositiveWidth=1,NegativeWidth=1,PositiveHeight=1,NegativeHeight=1)
    ActiveRegionSize=0.05
    Debug=(bDrawDebug=false,bDraw2D=False,bDrawLocalAxes=False,DebugScale=25,DebugSegments=20,DebugThickness=0.25)
    DriverAxis=(X=-1,Y=0,Z=0)
    DriverItem=(Type=Bone,Name="thigh_l")
    FalloffRegionScaleFactors=(PositiveWidth=1,NegativeWidth=1,PositiveHeight=1,NegativeHeight=1)
    FalloffSize=0.55
    FlipHeightScaling=False
    FlipWidthScaling=False
    OptionalParentItem=(Type=Bone,Name="None")
    RotationOffset=(X=90,Y=0,Z=-15)
    ExecutePin=@RigVMFunction_Sequence.B
   )
   > &RigUnit_ModifyTransforms(
    ItemToModify="((Item=(Type=Bone,Name=\"base_l_skirt_A\"),Transform=(Rotation=(X=-0.1607,Y=-0.6707,Z=0.4007,W=0.6032),Translation=(X=-10.2316,Y=1.3933,Z=-0.4354),Scale3D=(X=1,Y=1,Z=1))),(Item=(Type=Bone,Name=\"base_l_skirt_B\"),Transform=(Rotation=(X=0,Y=0,Z=0.5,W=0.866),Tra ...(truncated)"
    WeightMinimum=0
    Weight=@n5.OutputParam
   )
   > &RigUnit_ModifyTransforms(
    ItemToModify="((Item=(Type=Bone,Name=\"base_belt_front_l\"),Transform=(Rotation=(X=0.1128,Y=-0.0864,Z=0.5977,W=0.789),Translation=(X=8.7924,Y=-0.9556,Z=-0.1114),Scale3D=(X=1,Y=1,Z=1))),(Item=(Type=Bone,Name=\"base_belt_side_l\"),Transform=(Rotation=(X=0,Y=0.1852,Z=0,W=-0.98 ...(truncated)"
    WeightMinimum=0
    Weight=@n5.OutputParam
   )
   > n6 = &RigUnit_SphericalPoseReader(
    ActiveRegionScaleFactors=(PositiveWidth=1,NegativeWidth=1,PositiveHeight=1,NegativeHeight=1)
    ActiveRegionSize=0.05
    Debug=(bDrawDebug=false,bDraw2D=False,bDrawLocalAxes=False,DebugScale=25,DebugSegments=20,DebugThickness=0.25)
    DriverAxis=(X=-1,Y=0,Z=0)
    DriverItem=(Type=Bone,Name="thigh_l")
    FalloffRegionScaleFactors=(PositiveWidth=1,NegativeWidth=1,PositiveHeight=1,NegativeHeight=1)
    FalloffSize=0.55
    FlipHeightScaling=False
    FlipWidthScaling=False
    OptionalParentItem=(Type=Bone,Name="None")
    RotationOffset=(X=140,Y=20,Z=0)
   )
   > &RigUnit_ModifyTransforms(
    ItemToModify="((Item=(Type=Bone,Name=\"base_l_skirt_A\"),Transform=(Rotation=(X=0.0029,Y=-0.0352,Z=-0.0514,W=0.998),Translation=(X=-0.6188,Y=0.0085,Z=1.6172),Scale3D=(X=1,Y=1,Z=1))),(Item=(Type=Bone,Name=\"base_l_skirt_B\"),Transform=(Rotation=(X=-0.1536,Y=-0.1259,Z=0.0149, ...(truncated)"
    WeightMinimum=0
    Weight=@n6.OutputParam
   )
   > n7 = &RigUnit_SphericalPoseReader(
    ActiveRegionScaleFactors=(PositiveWidth=1,NegativeWidth=1,PositiveHeight=1,NegativeHeight=1)
    ActiveRegionSize=0.05
    Debug=(bDrawDebug=false,bDraw2D=False,bDrawLocalAxes=False,DebugScale=25,DebugSegments=20,DebugThickness=0.25)
    DriverAxis=(X=-1,Y=0,Z=0)
    DriverItem=(Type=Bone,Name="thigh_l")
    FalloffRegionScaleFactors=(PositiveWidth=1,NegativeWidth=1,PositiveHeight=1,NegativeHeight=1)
    FalloffSize=0.55
    FlipHeightScaling=False
    FlipWidthScaling=False
    OptionalParentItem=(Type=Bone,Name="None")
    RotationOffset=(X=180,Y=10,Z=0)
   )
   > &RigUnit_ModifyTransforms(
    ItemToModify="((Item=(Type=Bone,Name=\"offset_l_skirt_B\"),Transform=(Rotation=(X=-0.1874,Y=-0.6708,Z=0.1977,W=0.6898),Translation=(X=-7.0359,Y=-1.6065,Z=4.5451),Scale3D=(X=1,Y=1,Z=1))),(Item=(Type=Bone,Name=\"offset_l_skirt_A\"),Transform=(Rotation=(X=-0.2571,Y=-0.4511,Z=- ...(truncated)"
    WeightMinimum=0
    Weight=@n7.OutputParam
   )
   > &RigUnit_ModifyTransforms(
    ItemToModify="((Item=(Type=Bone,Name=\"base_belt_front_l\"),Transform=(Rotation=(X=0,Y=-0.1785,Z=0,W=-0.9839),Translation=(X=0.3351,Y=0,Z=0.4104),Scale3D=(X=1,Y=1,Z=1))),(Item=(Type=Bone,Name=\"base_belt_side_l\"),Transform=(Rotation=(X=-0.093,Y=0.1074,Z=-0.7575,W=-0.6372), ...(truncated)"
    WeightMinimum=0
    Weight=@n7.OutputParam
   )
   > n8 = &RigUnit_SphericalPoseReader(
    ActiveRegionScaleFactors=(PositiveWidth=1,NegativeWidth=1,PositiveHeight=1,NegativeHeight=1)
    ActiveRegionSize=0.05
    Debug=(bDrawDebug=false,bDraw2D=False,bDrawLocalAxes=False,DebugScale=25,DebugSegments=20,DebugThickness=0.25)
    DriverAxis=(X=-1,Y=0,Z=0)
    DriverItem=(Type=Bone,Name="thigh_l")
    FalloffRegionScaleFactors=(PositiveWidth=1,NegativeWidth=1,PositiveHeight=1,NegativeHeight=1)
    FalloffSize=0.55
    FlipHeightScaling=False
    FlipWidthScaling=False
    OptionalParentItem=(Type=Bone,Name="None")
    RotationOffset=(X=-90,Y=0,Z=20)
   )
   > &RigUnit_ModifyTransforms(
    ItemToModify="((Item=(Type=Bone,Name=\"base_l_skirt_C\"),Transform=(Rotation=(X=0.0177,Y=-0.7184,Z=-0.2149,W=0.6614),Translation=(X=-7.7893,Y=-1.511,Z=-1.2419),Scale3D=(X=1,Y=1,Z=1))),(Item=(Type=Bone,Name=\"base_l_skirt_B\"),Transform=(Rotation=(X=-0.0586,Y=-0.118,Z=-0.548 ...(truncated)"
    WeightMinimum=0
    Weight=@n8.OutputParam
   )
   RigVMFunctionReturnNode(ExecuteContext=@RigVMFunction_Sequence.C)
   Maximum_1 = RigVMAggregateNode(D=@n6.OutputParam, C=@n2.OutputParam, B=@n5.OutputParam, A=@n1.OutputParam)
   &RigUnit_ModifyTransforms(
    ItemToModify=((Item=(Type=Bone,Name="base_c_skirt_A"),Transform=(Rotation=(X=0,Y=-0.6428,Z=0,W=0.766),Translation=(X=-10,Y=0,Z=0),Scale3D=(X=1,Y=1,Z=1))))
    WeightMinimum=0
    Weight=@Maximum_1.Result
    ExecutePin=@RigVMFunction_Sequence.D
   )
   > &RigUnit_ModifyTransforms(
    ItemToModify=((Item=(Type=Bone,Name="base_c_skirt_B"),Transform=(Rotation=(X=0,Y=-0.766,Z=0,W=0.6428),Translation=(X=-6,Y=0,Z=-2),Scale3D=(X=1,Y=1,Z=1))))
    WeightMinimum=0
    Weight=&RigVMFunction_MathDoubleMax(B=@n8.OutputParam, A=@n4.OutputParam).Result
   )
  ]}
 New Function_ContainedGraph {nodes [
   RigVMFunctionEntryNode()
   > &RigUnit_ParentConstraint(
    AdvancedSettings=(RotationOrderForFilter=XZY)
    Child=(Type=Bone,Name="base_nade_01")
    Parents="((Item=(Type=Bone,Name=\"dyn_belt_back_r\"),Weight=0.0002),(Item=(Type=Bone,Name=\"dyn_belt_side_l\"),Weight=0.0053),(Item=(Type=Bone,Name=\"dyn_belt_side_r\"),Weight=0.0783),(Item=(Type=Bone,Name=\"dyn_belt_front_l\"),Weight=0.309),(Item=(Type=Bone,Name=\"dyn_belt ...(truncated)"
    Weight=1
   )
   > &RigUnit_ParentConstraint(
    AdvancedSettings=(RotationOrderForFilter=XZY)
    Child=(Type=Bone,Name="base_nade_02")
    Parents="((Item=(Type=Bone,Name=\"dyn_belt_side_l\"),Weight=0.0001),(Item=(Type=Bone,Name=\"dyn_belt_back_r\"),Weight=0.007),(Item=(Type=Bone,Name=\"dyn_belt_front_l\"),Weight=0.0375),(Item=(Type=Bone,Name=\"dyn_belt_side_r\"),Weight=0.4293),(Item=(Type=Bone,Name=\"dyn_belt ...(truncated)"
    Weight=1
   )
   > &RigUnit_ParentConstraint(
    AdvancedSettings=(RotationOrderForFilter=XZY)
    Child=(Type=Bone,Name="base_nade_03")
    Parents=((Item=(Type=Bone,Name="dyn_belt_front_l"),Weight=0.0049),(Item=(Type=Bone,Name="dyn_belt_back_r"),Weight=0.0404),(Item=(Type=Bone,Name="dyn_belt_front_r"),Weight=0.2498),(Item=(Type=Bone,Name="dyn_belt_side_r"),Weight=0.7049))
    Weight=1
   )
   > &RigUnit_ParentConstraint(
    AdvancedSettings=(RotationOrderForFilter=XZY)
    Child=(Type=Bone,Name="base_nade_04")
    Parents="((Item=(Type=Bone,Name=\"dyn_belt_side_l\"),Weight=0.0004),(Item=(Type=Bone,Name=\"dyn_belt_front_r\"),Weight=0.0007),(Item=(Type=Bone,Name=\"dyn_belt_back_l\"),Weight=0.0709),(Item=(Type=Bone,Name=\"dyn_belt_side_r\"),Weight=0.1664),(Item=(Type=Bone,Name=\"dyn_bel ...(truncated)"
    Weight=1
   )
   > &RigUnit_ParentConstraint(
    AdvancedSettings=(RotationOrderForFilter=XZY)
    Child=(Type=Bone,Name="base_knee_pad")
    Parents=((Item=(Type=Bone,Name="thigh_r"),Weight=0.1304),(Item=(Type=Bone,Name="deform_knee_out_r"),Weight=0.4104),(Item=(Type=Bone,Name="calf_r"),Weight=0.4592))
    Weight=1
   )
   > &RigUnit_ModifyTransforms(
    ItemToModify=((Item=(Type=Bone,Name="base_knee_pad"),Transform=(Rotation=(X=0.0543,Y=-0.0245,Z=0.1319,W=0.9895),Translation=(X=0.9679,Y=0.3126,Z=-0.5653),Scale3D=(X=1,Y=1,Z=1))))
    WeightMinimum=0
    Weight=RigVMDispatchNode(BoneName=root, Default=1, Name=deformRig_r_knee_alpha) {template="DISPATCH_RigDispatch_GetAnimAttribute(in Name,in BoneName,in Default,out Value,out Found)"}.Value
   )
   > RigVMFunctionReturnNode()
  ]}
 RigVMFunctionLibrary {nodes [
   Belt_Constraints = RigVMCollapseNode(ExecuteContext=())
   Skirt_SPR = RigVMCollapseNode(ExecuteContext=())
  ]}
 RigVMFunction_Sequence_1_ContainedGraph {nodes [
   n1 = RigVMFunctionEntryNode()
   RigVMFunction_Sequence = &RigVMFunction_Sequence(ExecuteContext=@n1.ExecuteContext)
   n2 = &RigVMFunction_Sequence(ExecuteContext=@RigVMFunction_Sequence.B)
   n3 = &RigVMFunction_Sequence(ExecuteContext=@n2.B)
   n4 = &RigVMFunction_Sequence(ExecuteContext=@n3.B)
   RigVMFunctionReturnNode(A=@RigVMFunction_Sequence.A, B=@n2.A, C=@n3.A, D=@n4.A, E=@n4.B)
  ]}
 RigVMModel {nodes [
   n1 = &RigUnit_PrepareForExecution()
   &RigUnit_HierarchyImportFromSkeleton(NameSpace=None, bIncludeCurves=False, bIncludeMeshSockets=False, bIncludeVirtualBones=True, ExecutePin=@n1.ExecutePin)
   > /Script/ControlRig.RigUnit_HierarchyAddBone(
    Name=thigh_l_ref
    Parent=(Type=Bone,Name="pelvis")
    Space=LocalSpace
    Transform=(Rotation=(X=-0.0009,Y=-0.0355,Z=-0.0155,W=0.9992),Translation=(X=-1.0855,Y=-0.2109,Z=-8.7053),Scale3D=(X=1,Y=1,Z=1))
   ) {is=struct RigVMUnitNode}
   > &RigUnit_HierarchyAddBone(
    Name=thigh_r_ref
    Parent=(Type=Bone,Name="pelvis")
    Space=LocalSpace
    Transform=(Rotation=(X=0.0355,Y=-0.0009,Z=0.9992,W=0.0155),Translation=(X=-1.0855,Y=-0.2109,Z=8.7053),Scale3D=(X=1,Y=1,Z=1))
   )
   RigUnit_BeginExecution = &RigUnit_BeginExecution()
   FN_DeformRig_FNC = RigVMFunctionReferenceNode(
    DeformRig_ Curve="(deformRig_l_upperarm_up_alpha_72_#1=False,deformRig_r_upperarm_up_alpha_73_#2=False,deformRig_l_upperarm_fr_alpha_74_#3=False,deformRig_r_upperarm_fr_alpha_75_CB51B8 ...(truncated)"
    DeformRig_Bool="(upperarm_1_#4=True,elbow_3_#5=True,wrist_7_#6=True,groin_8_#7=True,glute_17_#8=True,knee_9_40C408324CA261 ...(truncated)"
    DeformRig_Debug=(upperarm_debug_18_#4=False,wrist_debug_20_#6=False,lat_debug_26_#9=False,groinGlute_debug_27_#7=False)
    DeformRig_Value="(upperarm_up_18_#10=(Rotation=(X=0,Y=0,Z=0,W=1),Translation=(X=0,Y=0,Z=8),Scale3D=(X=0.7,Y=1,Z=1)),upperarm_fwd_fr_19_#11=(Rotation=(X=0,Y=0,Z=0,W=1),Translation=(X=0,Y=10,Z=0),Scale3D=(X=0.5,Y=1,Z= ...(truncated)"
    ExecuteContext=@RigUnit_BeginExecution.ExecutePin
   )
   > Skirt_SPR = RigVMFunctionReferenceNode()
   > Belt_Constraints = RigVMFunctionReferenceNode()
  ]}
]
variables [leg_l_null_transform {type=FTransform, default=(Rotation=(X=0,Y=0,Z=0,W=1),Translation=(X=100,Y=0,Z=89.0697),Scale3D=(X=1,Y=1,Z=1))}]
