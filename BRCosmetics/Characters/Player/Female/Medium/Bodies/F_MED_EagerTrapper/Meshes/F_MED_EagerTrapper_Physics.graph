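# Physics asset for the F_MED_EagerTrapper female player body, covering eight skirt bones and both thighs.
# Each dyn skirt capsule is constrained to its own tiny base or offset capsule at center, left or right, with narrow swing and zero-degree twist limits.
# Larger thigh capsules have no constraints, apparently serving as leg colliders for the skirt.
# Limits are asymmetric: right skirt B and C swing less than their left counterparts.
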PhysicsAsset /BRCosmetics/Characters/Player/Female/Medium/Bodies/F_MED_EagerTrapper/Meshes/F_MED_EagerTrapper_Physics
bodies {
 base_c_skirt_A [{shapeType=Capsule, length=0.505, radius=0.505, shapeName=base_c_skirt_A_capsule, rotation=<rot 0 -90 0>}]
 base_c_skirt_B [{shapeType=Capsule, length=0.505, radius=0.505, shapeName=base_c_skirt_B_capsule, rotation=<rot 0 -90 0>}]
 dyn_c_skirt_A [{shapeType=Capsule, length=11.5528, radius=2.0952, shapeName=dyn_c_skirt_A_capsule, center=<10 0 -2.2172>, rotation=<rot -90 90 0>}]
 dyn_c_skirt_B [{shapeType=Capsule, length=11.5528, radius=2.0952, shapeName=dyn_c_skirt_B_capsule, center=<11.1638 0 0.901>, rotation=<rot -90 90 0>}]
 dyn_l_skirt_A [{
   shapeType=Capsule
   length=11.5528
   radius=2.0952
   shapeName=dyn_l_skirt_A_capsule
   center=<10 -0.3458 0.764>
   rotation=<rot -73.0978 -89.9991 179.9993>
  }]
 dyn_l_skirt_B [{shapeType=Capsule, length=11.5528, radius=2.0952, shapeName=dyn_l_skirt_B_capsule, center=<10 0 2>, rotation=<rot -88 -90 180>}]
 dyn_l_skirt_C [{
   shapeType=Capsule
   length=11.5528
   radius=2.0952
   shapeName=dyn_l_skirt_C_capsule
   center=<10 1.428 1>
   rotation=<rot -88.6726 90.0168 -0.0166>
  }]
 dyn_r_skirt_A [{
   shapeType=Capsule
   length=11.5528
   radius=2.0952
   shapeName=dyn_l_skirt_A_capsule
   center=<-10 0.3458 -0.764>
   rotation=<rot -73.0978 -89.9991 179.9993>
  }]
 dyn_r_skirt_B [{shapeType=Capsule, length=11.5528, radius=2.0952, shapeName=dyn_l_skirt_B_capsule, center=<-10 0 -2>, rotation=<rot -88 -90 180>}]
 dyn_r_skirt_C [{
   shapeType=Capsule
   length=11.5528
   radius=2.0952
   shapeName=dyn_l_skirt_C_capsule
   center=<-10 -1.428 -1>
   rotation=<rot -88.6726 90.0168 -0.0166>
  }]
 offset_l_skirt_A [{shapeType=Capsule, length=0.505, radius=0.505, shapeName=offset_l_skirt_A_capsule, rotation=<rot 0 -90 0>}]
 offset_l_skirt_B [{shapeType=Capsule, length=0.505, radius=0.505, shapeName=offset_l_skirt_B_capsule, rotation=<rot 0 -90 0>}]
 offset_l_skirt_C [{shapeType=Capsule, length=0.505, radius=0.505, shapeName=offset_l_skirt_C_capsule, rotation=<rot 0 -90 0>}]
 offset_r_skirt_A [{shapeType=Capsule, length=0.505, radius=0.505, shapeName=offset_r_skirt_A_capsule, rotation=<rot 0 -90 0>}]
 offset_r_skirt_B [{shapeType=Capsule, length=0.505, radius=0.505, shapeName=offset_r_skirt_B_capsule, rotation=<rot 0 -90 0>}]
 offset_r_skirt_C [{shapeType=Capsule, length=0.505, radius=0.505, shapeName=offset_r_skirt_C_capsule, rotation=<rot 0 -90 0>}]
 thigh_l [{
   shapeType=Capsule
   length=15.5246
   radius=8.2923
   shapeName=thigh_r_capsule
   center=<-11.2881 -0.4277 0.0509>
   rotation=<rot 0.0829 -94.3723 -91.9077>
  }]
 thigh_r [{
   shapeType=Capsule
   length=15.5246
   radius=8.2923
   shapeName=thigh_r_capsule
   center=<11.2881 0.4277 -0.0509>
   rotation=<rot 0.0829 -94.3723 -91.9077>
  }]
}
constraints [
 {bone1Name=dyn_c_skirt_A, bone2Name=base_c_skirt_A, swing1LimitDegrees=7, swing2LimitDegrees=20, twistLimitDegrees=0}
 {bone1Name=dyn_c_skirt_B, bone2Name=base_c_skirt_B, swing1LimitDegrees=7, swing2LimitDegrees=12, twistLimitDegrees=0}
 {bone1Name=dyn_l_skirt_A, bone2Name=offset_l_skirt_A, swing1LimitDegrees=10, swing2LimitDegrees=20, twistLimitDegrees=0}
 {bone1Name=dyn_l_skirt_B, bone2Name=offset_l_skirt_B, swing1LimitDegrees=10, swing2LimitDegrees=20, twistLimitDegrees=0}
 {bone1Name=dyn_l_skirt_C, bone2Name=offset_l_skirt_C, swing1LimitDegrees=10, swing2LimitDegrees=20, twistLimitDegrees=0}
 {bone1Name=dyn_r_skirt_A, bone2Name=offset_r_skirt_A, swing1LimitDegrees=10, swing2LimitDegrees=20, twistLimitDegrees=0}
 {bone1Name=dyn_r_skirt_B, bone2Name=offset_r_skirt_B, swing1LimitDegrees=3, swing2LimitDegrees=8, twistLimitDegrees=0}
 {bone1Name=dyn_r_skirt_C, bone2Name=offset_r_skirt_C, swing1LimitDegrees=3, swing2LimitDegrees=8, twistLimitDegrees=0}
]
